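; Battle Pass quest gameplay ability that credits the player for launching while at a Crash Site.
; On event activation it calls the parent, then checks the owning pawn's gameplay tags for an exact match with any of the Crash Site tags.
; A match sends a custom stat event with configured source tags and a count of one to the player's quest manager.
; Both branches end the ability, with debug prints gated by a flag.
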(event Ability|EventActivateAbilityFromEvent
  (params (EventData :type Gameplay%20Event%20Data%20Structure%20%28by%20ref%29))
  (|CallParentFunction EventData)
  (Development|PrintString "@@@ ActivateAbilityFromEvent [GA_Quest_Launch_At_Crash_Site]: CALLED!" false (Variables|Debug|GetbDebugAbility))
  (if (GameplayTags|HasAnyTags (GameplayTags|GetOwnedGameplayTags (Variables|GAVars|Data|GetOwningPlayerPawn) :!class /Script/GameplayTags.GameplayTagAssetInterface) (Variables|Tags|GetT_CrashSiteTags) true)
    (Development|PrintString "@@@@@@ ActivateAbilityFromEvent [GA_Quest_Launch_At_Crash_Site]: We are launching at a Crash Site!" false (Variables|Debug|GetbDebugAbility))
    (Quest|SendComplexCustomStatEvent (Variables|GAVars|Data|GetOwningPlayerQuestManager) 0 (Variables|Tags|GetStatEventSourceTags) "(GameplayTags=)" 1)
    (Ability|EndAbility)
    (else
      (Development|PrintString "@@@@@@ ActivateAbilityFromEvent [GA_Quest_Launch_At_Crash_Site]: Not at a Crash Site..." false (Variables|Debug|GetbDebugAbility))
      (Ability|EndAbility))))
